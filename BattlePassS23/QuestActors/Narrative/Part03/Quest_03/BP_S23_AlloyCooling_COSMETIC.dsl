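; Cosmetic alloy-cooling actor for a Battle Pass S23 quest that animates a heated alloy mesh, then destroys itself.
; BeginPlay creates a dynamic material instance, stores the mesh's initial location and rotation, spawns an attached sound and starts a timeline.
; The timeline drives the mesh's relative height, rotation and scale, plus its heat transition through Custom Primitive Data slot 0.
; The stored initial transform values are not read in this graph.
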(event EventBeginPlay
  (bind _sm_alloyheated (Variables|Default|GetSM_AlloyHeated))
  (bind _returnvalue (Rendering|Material|CreateDynamicMaterialInstance (Variables|Default|GetSM_AlloyHeated) 0 (Rendering|Material|GetMaterial (Variables|Default|GetSM_AlloyHeated) 0) "None" :!class /Script/Engine.PrimitiveComponent))
  (Variables|Default|SetHeatTransitionDynamicMaterialInstance _returnvalue)
  (Variables|Default|SetInitialPosition (Variables|/Script/Engine.SceneComponent|GetRelativeLocation _sm_alloyheated))
  (Variables|Default|SetInitialRotation (Variables|/Script/Engine.SceneComponent|GetRelativeRotation _sm_alloyheated))
  (Audio|SpawnSoundAttached 0 (Variables|Default|GetSM_AlloyHeated))
  (bind (_direction _mortarshellsheight _mortarshellsrotation _mortarshellsscale _heattransitioncontrol) (Timeline|Timeline PlayFromStart)
    (:Update
      (Transformation|SetRelativeTransform (Variables|Default|GetSM_AlloyHeated) (Math|Transform|MakeTransform (Math|Vector|MakeVector 0.0 0.0 _mortarshellsheight) (Math|Rotator|MakeRotator (.x _mortarshellsrotation) (.y _mortarshellsrotation) (.z _mortarshellsrotation)) _mortarshellsscale) false false)
      (Rendering|Material|SetCustomPrimitiveDataFloat (Variables|Default|GetSM_AlloyHeated) 0 _heattransitioncontrol))
    (:Finished
      (Actor|DestroyActor))))

(fn ConstructionScript)
