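; Quest prop actor for the Season 23 narrative Anvil Hammer quest, handling the visual payoff when the quest succeeds.
; The quest component's success event passes the player's controlled pawn to EndVisuals, which runs only if that pawn is locally controlled.
; After a short delay it hides the minimap icon, swaps the pre-interaction mesh to malachite metal loot, plays a reward sound and activates a forge fire burst.
; After a longer delay it hides the whole actor.
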(event EventBeginPlay
  (|CallParentFunction) ; This node is disabled and will not be called.\nDrag off pins to build functionality.
)

(event Collision|EventActorBeginOverlap
  (params (OtherActor :type Actor%20Object%20Reference))
  (|CallParentFunction OtherActor) ; This node is disabled and will not be called.\nDrag off pins to build functionality.
)

(event EventTick
  (params (DeltaSeconds :type Float%20%28single-precision%29))
  (|CallParentFunction DeltaSeconds) ; This node is disabled and will not be called.\nDrag off pins to build functionality.
)

(event OnQuestComponentSuccessEvent(ConsolidatedQuestComponent)
  (params (PlayerController :type Fort%20Player%20Controller%20Object%20Reference))
  (CallFunction|EndVisuals (Pawn|GetControlledPawn PlayerController)))

(event Custom|EndVisuals
  (params (QuestOwnerPawn :type Pawn%20Object%20Reference))
  (bind _n_ambient (Variables|Components|GetN_Ambient))
  (if (Pawn|IsLocallyControlled QuestOwnerPawn)
    (Utilities|FlowControl|Delay 0.1)
    (MiniMapData|SetVisibility (Variables|BuildingGameplayActorPropQuest|GetQuestIconComponent) false)
    (Components|StaticMesh|SetStaticMesh (Variables|Components|GetSM_Quest_PreInteraction) "/Game/Items/Art_noLOD/Meshes/S_Loot_Metal_Malachite.S_Loot_Metal_Malachite")
    (Audio|PlaySound2D "/Game/Athena/Sounds/Quests/Athena_CollectHiddenStarReward_Cue.Athena_CollectHiddenStarReward_Cue" 1.0 1.0 0.0 0 QuestOwnerPawn)
    (Fortnite|SetNiagaraAssetMatchActivation _n_ambient "/BattlePassS23/QuestActors/Narrative/FX/P3Q3_Forge/NS_S23_P3Q3_Forge_Fire_Burst.NS_S23_P3Q3_Forge_Fire_Burst")
    (Components|Activation|Activate _n_ambient true)
    (Rendering|SetVisibility _n_ambient true false)
    (Utilities|FlowControl|Delay 2.5)
    (Rendering|SetVisibility (Variables|Transformation|GetRootComponent) false true)))

(fn ConstructionScript
  (|CallParentFunction))
