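# Material function that layers a heated-metal look, likely for a Battle Pass quest actor, over incoming material attributes.
# A sphere mask in pre-skinned position space grows from a start location as a custom primitive data control scales its radius.
# HeightLerp breaks up the mask using the material's roughness, and the result drives an alloy temper curve-atlas colour.
# That colour blends into base colour, and a boosted, powered copy blends into emissive through a parallel mask.
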MaterialFunction /BattlePassS23/QuestActors/Narrative/Part03/Quest_03/Materials/MF_HeatedMetal
graph {nodes [
  n1 = GetMaterialAttributes(None=FunctionInput() {inputName=Material, inputType=FunctionInput_MaterialAttributes}) {attributeGetTypes [#1, #2, #3]}
  n2 = NamedRerouteUsage()
  n3 = VectorParameter() {parameterName=HeatStartLoc, defaultValue=<rgba 0 0 0 1>}
  n4 = ScalarParameter() {defaultValue=64, parameterName=HeatRadius}
  n5 = ScalarParameter() {bUseCustomPrimitiveData=true, parameterName=HeatTransitionControl}
  n6 = ScalarParameter() {parameterName=HeatTransitionHardness}
  n7 = MaterialFunctionCall(Transition Phase=SphereMask(A=@n2, B=@n3.RGB, Radius=Multiply(A=@n4, B=@n5), Hardness=@n6), Height Texture=@n1.Roughness) {materialFunction=&HeightLerp}
  n8 = CurveAtlasRowParameter(CurveTime=@n7.Lerp Alpha No Contrast) {
   atlas=&/AlloyCurve
   curve=&/AlloyTemperColor
   parameterName=AlloyCurve
  }
  n9 = Power(
   Base=Multiply(A=@n8, B=ScalarParameter() {defaultValue=4, parameterName=HeatEmissiveStrength})
   Exp=ScalarParameter() {defaultValue=2, parameterName=HeatEmissivePower}
  )
  n10 = Multiply(A=@n5)
  n11 = MaterialFunctionCall(Transition Phase=SphereMask(A=@n2, B=@n3.RGB, Radius=Multiply(A=@n4, B=@n10), Hardness=@n6), Height Texture=@n1.Roughness) {materialFunction=&HeightLerp}
  FunctionOutput(None=SetMaterialAttributes!(
    MaterialAttributes=@n1.MaterialAttributes
    EmissiveColor=LinearInterpolate(A=@n1.EmissiveColor, B=@n9, Alpha=@n11.Lerp Alpha No Contrast)
    BaseColor=LinearInterpolate(A=@n1.BaseColor, B=@n8, Alpha=@n7.Lerp Alpha No Contrast)
   ) {attributeSetTypes [#2, #3]})
  n12 = PreSkinnedPosition()
  NamedRerouteDeclaration(None=ShaderStageSwitch(PixelShader=VertexInterpolator(VS=@n12).PS, VertexShader=@n12)) {name=PreskinnedPos, nodeColor=<rgba 1 0 0.5156 1>}
 ]}
